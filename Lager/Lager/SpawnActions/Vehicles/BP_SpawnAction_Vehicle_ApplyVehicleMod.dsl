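; Living World spawn action that randomly fits tire and cowcatcher mods to newly spawned vehicles.
; OnAllActorSpawned loops over the spawned actors, casts each to the Fort vehicle interface and skips inoperable vehicles.
; Each mod has its own configurable tag and chance, and is applied only if the roll passes and CanApplyMod allows it.
; Vehicles tagged as Valet semis never receive the tire mod but remain eligible for the cowcatcher.
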(fn OnAllActorSpawned
  (params (SpawnedActor :type Array%20of%20Actor%20Object%20References) (SourcePointProvider :type Fort%20Athena%20Living%20World%20Point%20Provider%20Interface%20Interface%20%28by%20ref%29))
  (bind _cowcatchermodtag (Variables|Default|GetCowCatcherModTag))
  (bind _tiremodtag (Variables|Default|GetTireModTag))
  (for _array_element SpawnedActor
    (bind _asfort_vehicle_interface (Utilities|Casting|CastToFortVehicleInterface _array_element))
    (if (not (Vehicle|IsInoperable _asfort_vehicle_interface))
      ;[1 1 1] Apply Tire Mod
      (if (Math|Boolean|ANDBoolean (Utilities|GenerateRandomChance (Variables|Default|GetChanceToApplyTireMod)) (Vehicle|Mod|CanApplyMod _asfort_vehicle_interface _tiremodtag) (not (GameplayTags|HasTag (Class|FortVehicleInterface|GetVehicleTags _asfort_vehicle_interface) "(TagName=\"Vehicle.Valet.Semi\")")))
        (Vehicle|Mod|ApplyVehicleMod _asfort_vehicle_interface _tiremodtag 0)
        (else
          _))
      (if (and (Utilities|GenerateRandomChance (Variables|Default|GetChanceToApplyCowCatcher)) (Vehicle|Mod|CanApplyMod _asfort_vehicle_interface _cowcatchermodtag))
        (Vehicle|Mod|ApplyVehicleMod _asfort_vehicle_interface _cowcatchermodtag 0))))
  (return))
