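; Living World spawn action that can make freshly spawned Valet vehicles (drivable cars and trucks) inoperable.
; On all actors spawned, it exits if the all-operable scalable bool is set; otherwise it casts each actor to the Fort vehicle interface.
; An all-inoperable bool disables every vehicle; without it, a gameplay tag switch rolls a separate chance for basic sedan, pickup, semi, sport and taxi before calling MakeInoperable.
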(fn OnAllActorSpawned
  (params (SpawnedActor :type Array%20of%20Actor%20Object%20References) (SourcePointProvider :type Fort%20Athena%20Living%20World%20Point%20Provider%20Interface%20Interface%20%28by%20ref%29))
  (if (ScalableFloat|GetValueasBool (Variables|Default|GetAllValetOperable))
    (return)
    (else
      (for _array_element SpawnedActor
        (bind _asfort_vehicle_interface (Utilities|Casting|CastToFortVehicleInterface _array_element))
        ;[1 1 1] Set Vehicle Inoperable
        (if (not (ScalableFloat|GetValueasBool (Variables|Default|GetAllValetInoperable)))
          (switch Utilities|FlowControl|Switch|SwitchonGameplayTagContainer (Class|FortVehicleInterface|GetVehicleTags _asfort_vehicle_interface)
            (:(GameplayTags=((TagName="Vehicle.Valet.BasicSedan")))
              (if (Utilities|GenerateRandomChance (Variables|Default|GetBasicCarInoperableChance))
                _))
            (:(GameplayTags=((TagName="Vehicle.Valet.Pickup")))
              (if (Utilities|GenerateRandomChance (Variables|Default|GetPickupInoperableChance))
                _))
            (:(GameplayTags=((TagName="Vehicle.Valet.Semi")))
              (if (Utilities|GenerateRandomChance (Variables|Default|GetSemiInoperableChance))
                _))
            (:(GameplayTags=((TagName="Vehicle.Valet.Sport")))
              (if (Utilities|GenerateRandomChance (Variables|Default|GetSportsCarInoperableChance))
                _))
            (:(GameplayTags=((TagName="Vehicle.Valet.Taxi")))
              (if (Utilities|GenerateRandomChance (Variables|Default|GetTaxiInoperableChance))
                _))))
        (Vehicle|MakeInoperable _asfort_vehicle_interface))
      (return))))
